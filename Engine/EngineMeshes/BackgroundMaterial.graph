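# BackgroundMaterial is a minimal Default Lit material in the engine's EngineMeshes folder, flagged for use with static lighting.
# Its graph holds a single Constant3Vector left at its default black, with no connection to any material input shown.
# It has no textures or parameters, and pixel depth offset is set to legacy mode.
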Material /Engine/EngineMeshes/BackgroundMaterial
graph {properties {bUsedWithStaticLighting=true, pixelDepthOffsetMode=PDOM_Legacy, usedShadingModels=Default Lit}, nodes [Constant3Vector()]}
