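# M_Mat_UnrealTech is a Default Lit engine material for meshes, flagged for instanced static meshes and static lighting.
# A grayscale sample of the T_Mat_UnrealTech_01_D texture is multiplied by a Brightness scalar parameter, so material instances can tune its intensity.
# Three fixed scalar constants sit alongside that chain, with their destination pins not shown.
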Material /Engine/EngineMeshes/M_Mat_UnrealTech
graph {
 properties {bUsedWithInstancedStaticMeshes=true, bUsedWithStaticLighting=true, pixelDepthOffsetMode=PDOM_Legacy, usedShadingModels=Default Lit}
 nodes [
  Constant() {r=0.3}
  Constant()
  Constant() {r=0.2}
  Multiply(
   A=TextureSample() {samplerType=SAMPLERTYPE_Grayscale, texture=&/T_Mat_UnrealTech_01_D}.RGB
   B=ScalarParameter() {defaultValue=0.8, parameterName=Brightness}
  )
 ]
}
